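# Physics asset for the front of the male medium Ebony Spark body cosmetic, covering two dangling front strap chains and the body around them.
# Each strap runs from a small MSTR anchor on spine_04 through four capsule links, free-rotating at the root joints and narrowly limited in swing and twist further down.
# Capsules on the pelvis, spine, clavicles, arms, thighs and calves form a simplified body, likely as collision for the straps.
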PhysicsAsset /BRCosmetics/Characters/Player/Male/Medium/Bodies/M_MED_Ebony_Spark/Meshes/M_MED_Ebony_Spark_Front_Physics
bodies {
 calf_l [{shapeType=Capsule, length=26.75, radius=7.25, center=<-18 0 0>, rotation=<rot 90 0 0>}]
 calf_r [{shapeType=Capsule, length=26.75, radius=7.25, center=<18 0 0>, rotation=<rot 90 0 0>}]
 clavicle_l [{shapeType=Capsule, length=6.9279, radius=7.7908, center=<11.504 0.8386 -1.0247>, rotation=<rot 90 5 0>}]
 clavicle_r [{shapeType=Capsule, length=6.9279, radius=7.7908, center=<-11.504 -0.8386 1.0247>, rotation=<rot 90 5 0>}]
 dyn_strap_fr_l_01 [{shapeType=Capsule, length=5, radius=4, shapeName=dyn_strap_fr_l_01_capsule, center=<6 0 0>, rotation=<rot 90 0 0>}]
 dyn_strap_fr_l_02 [{shapeType=Capsule, length=4.5, radius=4.5, shapeName=dyn_strap_fr_l_02_capsule, center=<6 3 0>, rotation=<rot 90 0 0>}]
 dyn_strap_fr_l_03 [{shapeType=Capsule, length=4.5, radius=4.5, shapeName=dyn_strap_fr_l_03_capsule, center=<6 3 0>, rotation=<rot 90 0 0>}]
 dyn_strap_fr_l_04 [{shapeType=Capsule, length=4.5, radius=4.5, shapeName=dyn_strap_fr_l_04_capsule, center=<6 3 0>, rotation=<rot 90 0 0>}]
 dyn_strap_fr_l_MSTR [{shapeType=Capsule, length=0.505, radius=0.505, shapeName=dyn_strap_fr_l_MSTR_capsule, rotation=<rot 0 -90 0>}]
 dyn_strap_fr_r_01 [{shapeType=Capsule, length=5, radius=4, shapeName=dyn_strap_fr_r_01_capsule, center=<6 0 0>, rotation=<rot 90 0 0>}]
 dyn_strap_fr_r_02 [{shapeType=Capsule, length=4.5, radius=4.5, shapeName=dyn_strap_fr_r_02_capsule, center=<6 3 0>, rotation=<rot 90 0 0>}]
 dyn_strap_fr_r_03 [{shapeType=Capsule, length=4.5, radius=4.5, shapeName=dyn_strap_fr_r_03_capsule, center=<6 3 0>, rotation=<rot 90 0 0>}]
 dyn_strap_fr_r_04 [{shapeType=Capsule, length=4.5, radius=4.5, shapeName=dyn_strap_fr_r_04_capsule, center=<6 3 0>, rotation=<rot 90 0 0>}]
 dyn_strap_fr_r_MSTR [{shapeType=Capsule, length=0.505, radius=0.505, shapeName=dyn_strap_fr_r_MSTR_capsule, rotation=<rot 0 -90 0>}]
 lowerarm_l [{shapeType=Capsule, length=37.6377, radius=4.63, center=<18.227 0.9178 -2.1609>, rotation=<rot 82.7121 -0.0002 3.2014>}]
 lowerarm_r [{shapeType=Capsule, length=30.5439, radius=5.63, center=<-14.5234 -1.1325 0.8591>, rotation=<rot 82.7121 -0.0002 3.2014>}]
 pelvis [{shapeType=Capsule, length=8.25, radius=11.25, center=<-3.0138 -1.7655 0>}]
 spine_01 [{shapeType=Capsule, length=11.25, radius=12.25}]
 spine_02 [{shapeType=Capsule, length=4.625, radius=13.625}]
 spine_03 [{shapeType=Capsule, length=2.875, radius=13.875, center=<0 0.7081 0>}]
 spine_04 [{shapeType=Capsule, length=3.625, radius=12, center=<0 1.1382 0>}]
 thigh_twist_01_l [{shapeType=Capsule, length=27.125, radius=6.875, center=<0 4.3162 -6.9267>, rotation=<rot 85 0 0>}]
 thigh_twist_01_r [{shapeType=Capsule, length=27.125, radius=6.875, center=<0 -4.912 6.1551>, rotation=<rot 85 0 -180>}]
 upperarm_twist_01_l [{shapeType=Capsule, length=24.1779, radius=5.0408, center=<13.2584 -0.8854 -1.0226>, rotation=<rot 87.3404 12.8837 -172.3975>}]
 upperarm_twist_01_r [{shapeType=Capsule, length=24.1779, radius=5.0408, center=<-13.2584 -0.8854 -1.0226>, rotation=<rot 87.3404 12.8837 -172.3975>}]
}
constraints [
 {bone1Name=spine_03, bone2Name=spine_02, swing1LimitDegrees=45, swing2LimitDegrees=45, twistLimitDegrees=45}
 {bone1Name=spine_04, bone2Name=spine_03, swing1LimitDegrees=45, swing2LimitDegrees=45, twistLimitDegrees=45}
 {bone1Name=clavicle_l, bone2Name=spine_04, swing1LimitDegrees=45, swing2LimitDegrees=45, twistLimitDegrees=45}
 {bone1Name=lowerarm_l, bone2Name=clavicle_l, swing1LimitDegrees=45, swing2LimitDegrees=45, twistLimitDegrees=45}
 {bone1Name=upperarm_twist_01_l, bone2Name=clavicle_l, swing1LimitDegrees=45, swing2LimitDegrees=45, twistLimitDegrees=45}
 {bone1Name=clavicle_r, bone2Name=spine_04, swing1LimitDegrees=45, swing2LimitDegrees=45, twistLimitDegrees=45}
 {bone1Name=lowerarm_r, bone2Name=clavicle_r, swing1LimitDegrees=45, swing2LimitDegrees=45, twistLimitDegrees=45}
 {bone1Name=upperarm_twist_01_r, bone2Name=clavicle_r, swing1LimitDegrees=45, swing2LimitDegrees=45, twistLimitDegrees=45}
 {bone1Name=calf_l, bone2Name=pelvis, swing1LimitDegrees=45, swing2LimitDegrees=45, twistLimitDegrees=45}
 {bone1Name=calf_r, bone2Name=pelvis, swing1LimitDegrees=45, swing2LimitDegrees=45, twistLimitDegrees=45}
 {
  bone1Name=spine_01
  bone2Name=pelvis
  swing1LimitDegrees=45
  swing1Motion=Free
  swing2LimitDegrees=45
  swing2Motion=Free
  twistLimitDegrees=45
  twistMotion=Free
 }
 {
  bone1Name=spine_02
  bone2Name=spine_01
  swing1LimitDegrees=45
  swing1Motion=Free
  swing2LimitDegrees=45
  swing2Motion=Free
  twistLimitDegrees=45
  twistMotion=Free
 }
 {bone1Name=dyn_strap_fr_r_04, bone2Name=dyn_strap_fr_r_03, swing1LimitDegrees=15, swing2LimitDegrees=15, twistLimitDegrees=5}
 {
  bone1Name=dyn_strap_fr_r_MSTR
  bone2Name=spine_04
  swing1LimitDegrees=45
  swing1Motion=Free
  swing2LimitDegrees=45
  swing2Motion=Free
  twistLimitDegrees=45
  twistMotion=Free
 }
 {
  bone1Name=dyn_strap_fr_r_01
  bone2Name=dyn_strap_fr_r_MSTR
  swing1LimitDegrees=45
  swing1Motion=Free
  swing2LimitDegrees=45
  swing2Motion=Free
  twistLimitDegrees=45
  twistMotion=Free
 }
 {bone1Name=dyn_strap_fr_r_02, bone2Name=dyn_strap_fr_r_01, swing1LimitDegrees=10, swing2LimitDegrees=10, twistLimitDegrees=5}
 {bone1Name=dyn_strap_fr_r_03, bone2Name=dyn_strap_fr_r_02, swing1LimitDegrees=15, swing2LimitDegrees=15, twistLimitDegrees=5}
 {bone1Name=dyn_strap_fr_l_04, bone2Name=dyn_strap_fr_l_03, swing1LimitDegrees=15, swing2LimitDegrees=15, twistLimitDegrees=5}
 {
  bone1Name=dyn_strap_fr_l_MSTR
  bone2Name=spine_04
  swing1LimitDegrees=45
  swing1Motion=Free
  swing2LimitDegrees=45
  swing2Motion=Free
  twistLimitDegrees=45
  twistMotion=Free
 }
 {
  bone1Name=dyn_strap_fr_l_01
  bone2Name=dyn_strap_fr_l_MSTR
  swing1LimitDegrees=45
  swing1Motion=Free
  swing2LimitDegrees=45
  swing2Motion=Free
  twistLimitDegrees=45
  twistMotion=Free
 }
 {bone1Name=dyn_strap_fr_l_02, bone2Name=dyn_strap_fr_l_01, swing1LimitDegrees=10, swing2LimitDegrees=10, twistLimitDegrees=5}
 {bone1Name=dyn_strap_fr_l_03, bone2Name=dyn_strap_fr_l_02, swing1LimitDegrees=15, swing2LimitDegrees=15, twistLimitDegrees=5}
 {
  bone1Name=thigh_twist_01_r
  bone2Name=pelvis
  swing1LimitDegrees=45
  swing1Motion=Free
  swing2LimitDegrees=45
  swing2Motion=Free
  twistLimitDegrees=45
  twistMotion=Free
 }
 {
  bone1Name=thigh_twist_01_l
  bone2Name=pelvis
  swing1LimitDegrees=45
  swing1Motion=Free
  swing2LimitDegrees=45
  swing2Motion=Free
  twistLimitDegrees=45
  twistMotion=Free
 }
]
