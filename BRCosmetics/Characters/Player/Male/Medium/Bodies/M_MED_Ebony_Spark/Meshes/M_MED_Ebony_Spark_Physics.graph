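# Physics asset for the M_MED_Ebony_Spark male player cosmetic body, set up for a dynamic jacket with lower-body and floor collision.
# Capsules cover pelvis, legs and feet, while a Floor_COL body adds a large sphere below its origin as an apparent ground proxy.
# Left and right back, side and front jacket chains hang from pelvis-linked master bodies with limited swing and twist.
# Angularly free cross-constraints tie neighboring chains together so the jacket holds together.
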PhysicsAsset /BRCosmetics/Characters/Player/Male/Medium/Bodies/M_MED_Ebony_Spark/Meshes/M_MED_Ebony_Spark_Physics
bodies {
 Floor_COL [
  {shapeType=Sphere, radius=187.1686, shapeName=root_sphere, center=<-3.4923 -2.714 -178.0361>}
  {shapeType=Sphere, radius=24.9585, shapeName=root_sphere, center=<0 2.4787 0.1791>}
 ]
 calf_l [{shapeType=Capsule, length=31.6821, radius=8.1093, center=<-25 -2 0>, rotation=<rot 90 0 0>}]
 calf_r [{shapeType=Capsule, length=30.0571, radius=9.6093, center=<24.8565 1.282 -3.7248>, rotation=<rot 85 0 0>}]
 dyn_jacketBack_01_l [{shapeType=Capsule, length=10, radius=10, shapeName=dyn_jacketBack_01_l_capsule, center=<7 0 9>, rotation=<rot 0 0 90>}]
 dyn_jacketBack_01_r [{shapeType=Capsule, length=10, radius=10, shapeName=dyn_jacketBack_01_l_capsule, center=<7 0 9>, rotation=<rot 0 0 90>}]
 dyn_jacketBack_02_l [{shapeType=Capsule, length=10, radius=9, shapeName=dyn_jacketBack_02_l_capsule, center=<6 0 8.5>, rotation=<rot -0.0004 0.0001 80>}]
 dyn_jacketBack_02_r [{shapeType=Capsule, length=10, radius=9, shapeName=dyn_jacketBack_02_l_capsule, center=<6 0 8.5>, rotation=<rot -0.0004 -0.0001 100>}]
 dyn_jacketBack_03_l [{shapeType=Capsule, length=16, radius=7.5, shapeName=dyn_jacketBack_03_l_capsule, center=<6 0 4>, rotation=<rot 0 0 90>}]
 dyn_jacketBack_03_r [{shapeType=Capsule, length=16.5, radius=7, shapeName=dyn_jacketBack_03_l_capsule, center=<6 -6 4>, rotation=<rot 0 0 90>}]
 dyn_jacketBack_04_l [{shapeType=Capsule, length=17.5, radius=8, shapeName=dyn_jacketBack_04_l_capsule, center=<6 0 4>, rotation=<rot 0 0 90>}]
 dyn_jacketBack_04_r [{shapeType=Capsule, length=17.75, radius=7.75, shapeName=dyn_jacketBack_04_l_capsule, center=<6 -6 4>, rotation=<rot 0 0 95>}]
 dyn_jacketBack_mstr_l [{shapeType=Capsule, length=0.505, radius=0.505, shapeName=dyn_jacketBack_mstr_l_capsule, rotation=<rot 0 -90 0>}]
 dyn_jacketBack_mstr_r [{shapeType=Capsule, length=0.505, radius=0.505, shapeName=dyn_jacketBack_mstr_r_capsule, rotation=<rot 0 -90 0>}]
 dyn_jacketFront_l_01 [{
   shapeType=Capsule
   length=6.2284
   radius=7.0216
   shapeName=dyn_jacketFront_l_01_capsule
   center=<5 -5 3>
   rotation=<rot 5.7165 34.5893 76.7812>
  }]
 dyn_jacketFront_l_02 [
  {
   shapeType=Capsule
   length=8.5372
   radius=5.7128
   shapeName=dyn_jacketFront_l_02_capsule
   center=<6 -5 3>
   rotation=<rot 5.7165 34.5893 76.7812>
  }
  {
   shapeType=Capsule
   length=11.0685
   radius=5.2931
   shapeName=dyn_jacketFront_l_02_capsule
   center=<14.1473 -5.8195 2.9015>
   rotation=<rot 5.0001 35.0001 75>
  }
 ]
 dyn_jacketFront_mstr_l [{shapeType=Capsule, length=0.505, radius=0.505, shapeName=dyn_jacketFront_mstr_l_capsule, rotation=<rot 0 -90 0>}]
 dyn_jacketFront_mstr_r [{shapeType=Capsule, length=0.505, radius=0.505, shapeName=dyn_jacketFront_mstr_r_capsule, rotation=<rot 0 -90 0>}]
 dyn_jacketFront_r_01 [{shapeType=Capsule, length=8, radius=5.25, shapeName=dyn_jacketFront_l_01_capsule, center=<5 5 2>, rotation=<rot 4.9812 -29.6219 98.6821>}]
 dyn_jacketFront_r_02 [
  {
   shapeType=Capsule
   length=8.625
   radius=5.625
   shapeName=dyn_jacketFront_l_02_capsule
   center=<6 5 2>
   rotation=<rot 4.9812 -29.6219 98.6821>
  }
  {shapeType=Capsule, length=4.625, radius=4.875, shapeName=dyn_jacketFront_r_02_capsule, center=<12.36 3.155 2.4236>, rotation=<rot 5 -30 100>}
 ]
 dyn_jacketSide_01_l [{shapeType=Capsule, length=8.6326, radius=8.3674, shapeName=dyn_jacketSide_01_l_capsule, center=<6 -4 6>, rotation=<rot 0 0.0001 70>}]
 dyn_jacketSide_01_r [{shapeType=Capsule, length=9.25, radius=8.5, shapeName=dyn_jacketSide_01_l_capsule, center=<6 3.4434 3.9226>, rotation=<rot 0 -0.0001 105>}]
 dyn_jacketSide_02_l [{shapeType=Capsule, length=11.2264, radius=8.1358, shapeName=dyn_jacketSide_02_l_capsule, center=<6 -4 5>, rotation=<rot 0 0.0004 74.9999>}]
 dyn_jacketSide_02_r [{shapeType=Capsule, length=9.75, radius=8.125, shapeName=dyn_jacketSide_02_l_capsule, center=<6 4 5>, rotation=<rot 0 -0.0001 105>}]
 dyn_jacketSide_03_l [{shapeType=Capsule, length=8.75, radius=8.25, shapeName=dyn_jacketSide_02_l_capsule, center=<6 -4 5>, rotation=<rot 0 0.0001 70>}]
 dyn_jacketSide_03_r [{shapeType=Capsule, length=8.875, radius=8.125, shapeName=dyn_jacketSide_02_l_capsule, center=<6 4 5>, rotation=<rot 0 -0.0001 105>}]
 dyn_jacketSide_04_l [{shapeType=Capsule, length=12.875, radius=8.125, shapeName=dyn_jacketSide_02_l_capsule, center=<15 -5 2.5>, rotation=<rot 85 -4.9999 -5.0003>}]
 dyn_jacketSide_04_r [{shapeType=Capsule, length=13.375, radius=7.625, shapeName=dyn_jacketSide_02_l_capsule, center=<15 5 2.5>, rotation=<rot 80 10 0.0002>}]
 dyn_jacketSide_mstr_l [{shapeType=Capsule, length=0.505, radius=0.505, shapeName=dyn_jacketSide_mstr_l_capsule, rotation=<rot 0 -90 0>}]
 dyn_jacketSide_mstr_r [{shapeType=Capsule, length=0.505, radius=0.505, shapeName=dyn_jacketSide_mstr_r_capsule, rotation=<rot 0 -90 0>}]
 foot_l [{shapeType=Capsule, length=15.3987, radius=9.8757, center=<-3.4271 8.491 0.1091>, rotation=<rot 0.7349 0.9586 -85.8278>}]
 foot_r [{shapeType=Capsule, length=16.7876, radius=9.1425, center=<3.4814 -8.3475 -2.0352>, rotation=<rot 0.7221 1.1909 -85.9009>}]
 pelvis [
  {shapeType=Capsule, length=8.2999, radius=10.8705, center=<-3.1833 -0.3011 0>}
  {shapeType=Capsule, length=7.5, radius=11, shapeName=pelvis_capsule, center=<-11.8037 -2.3855 0>}
 ]
 thigh_l [{shapeType=Capsule, length=31.6527, radius=8.4275, shapeName=thigh_l_capsule, center=<-24 1 0>, rotation=<rot 90 0 -1.6898>}]
 thigh_r [{shapeType=Capsule, length=30.9726, radius=8.4275, shapeName=thigh_r_capsule, center=<24 -1 0>, rotation=<rot 90 0 -1.6898>}]
}
constraints [
 {bone1Name=calf_l, bone2Name=thigh_l, swing1LimitDegrees=45, swing2LimitDegrees=45, twistLimitDegrees=45}
 {bone1Name=foot_l, bone2Name=calf_l, swing1LimitDegrees=45, swing2LimitDegrees=45, twistLimitDegrees=45}
 {bone1Name=calf_r, bone2Name=thigh_r, swing1LimitDegrees=45, swing2LimitDegrees=45, twistLimitDegrees=45}
 {bone1Name=foot_r, bone2Name=calf_r, swing1LimitDegrees=45, swing2LimitDegrees=45, twistLimitDegrees=45}
 {
  bone1Name=thigh_l
  bone2Name=pelvis
  swing1LimitDegrees=45
  swing1Motion=Free
  swing2LimitDegrees=45
  swing2Motion=Free
  twistLimitDegrees=45
  twistMotion=Free
 }
 {
  bone1Name=thigh_r
  bone2Name=pelvis
  swing1LimitDegrees=45
  swing1Motion=Free
  swing2LimitDegrees=45
  swing2Motion=Free
  twistLimitDegrees=45
  twistMotion=Free
 }
 {bone1Name=dyn_jacketBack_04_r, bone2Name=dyn_jacketBack_03_r, swing1LimitDegrees=35, swing2LimitDegrees=35, twistLimitDegrees=5}
 {bone1Name=dyn_jacketBack_01_r, bone2Name=dyn_jacketBack_mstr_r, swing1LimitDegrees=35, swing2LimitDegrees=35, twistLimitDegrees=5}
 {bone1Name=dyn_jacketBack_02_r, bone2Name=dyn_jacketBack_01_r, swing1LimitDegrees=35, swing2LimitDegrees=35, twistLimitDegrees=5}
 {bone1Name=dyn_jacketBack_03_r, bone2Name=dyn_jacketBack_02_r, swing1LimitDegrees=35, swing2LimitDegrees=35, twistLimitDegrees=5}
 {
  bone1Name=dyn_jacketBack_mstr_r
  bone2Name=pelvis
  swing1LimitDegrees=45
  swing1Motion=Free
  swing2LimitDegrees=45
  swing2Motion=Free
  twistLimitDegrees=45
  twistMotion=Free
 }
 {bone1Name=dyn_jacketBack_04_l, bone2Name=dyn_jacketBack_03_l, swing1LimitDegrees=35, swing2LimitDegrees=35, twistLimitDegrees=5}
 {
  bone1Name=dyn_jacketBack_mstr_l
  bone2Name=pelvis
  swing1LimitDegrees=45
  swing1Motion=Free
  swing2LimitDegrees=45
  swing2Motion=Free
  twistLimitDegrees=45
  twistMotion=Free
 }
 {bone1Name=dyn_jacketBack_01_l, bone2Name=dyn_jacketBack_mstr_l, swing1LimitDegrees=35, swing2LimitDegrees=35, twistLimitDegrees=5}
 {bone1Name=dyn_jacketBack_02_l, bone2Name=dyn_jacketBack_01_l, swing1LimitDegrees=35, swing2LimitDegrees=35, twistLimitDegrees=5}
 {bone1Name=dyn_jacketBack_03_l, bone2Name=dyn_jacketBack_02_l, swing1LimitDegrees=35, swing2LimitDegrees=35, twistLimitDegrees=5}
 {
  bone1Name=dyn_jacketFront_mstr_l
  bone2Name=pelvis
  swing1LimitDegrees=45
  swing1Motion=Free
  swing2LimitDegrees=45
  swing2Motion=Free
  twistLimitDegrees=45
  twistMotion=Free
 }
 {bone1Name=dyn_jacketFront_l_01, bone2Name=dyn_jacketFront_mstr_l, swing1LimitDegrees=35, swing2LimitDegrees=35, twistLimitDegrees=15}
 {bone1Name=dyn_jacketFront_l_02, bone2Name=dyn_jacketFront_l_01, swing1LimitDegrees=35, swing2LimitDegrees=35, twistLimitDegrees=15}
 {
  bone1Name=dyn_jacketFront_mstr_r
  bone2Name=pelvis
  swing1LimitDegrees=45
  swing1Motion=Free
  swing2LimitDegrees=45
  swing2Motion=Free
  twistLimitDegrees=45
  twistMotion=Free
 }
 {bone1Name=dyn_jacketFront_r_01, bone2Name=dyn_jacketFront_mstr_r, swing1LimitDegrees=35, swing2LimitDegrees=35, twistLimitDegrees=15}
 {bone1Name=dyn_jacketFront_r_02, bone2Name=dyn_jacketFront_r_01, swing1LimitDegrees=35, swing2LimitDegrees=35, twistLimitDegrees=15}
 {
  bone1Name=dyn_jacketSide_mstr_l
  bone2Name=pelvis
  swing1LimitDegrees=45
  swing1Motion=Free
  swing2LimitDegrees=45
  swing2Motion=Free
  twistLimitDegrees=45
  twistMotion=Free
 }
 {bone1Name=dyn_jacketSide_01_l, bone2Name=dyn_jacketSide_mstr_l, swing1LimitDegrees=35, swing2LimitDegrees=35, twistLimitDegrees=15}
 {bone1Name=dyn_jacketSide_02_l, bone2Name=dyn_jacketSide_01_l, swing1LimitDegrees=35, swing2LimitDegrees=35, twistLimitDegrees=15}
 {
  bone1Name=dyn_jacketSide_mstr_r
  bone2Name=pelvis
  swing1LimitDegrees=45
  swing1Motion=Free
  swing2LimitDegrees=45
  swing2Motion=Free
  twistLimitDegrees=45
  twistMotion=Free
 }
 {bone1Name=dyn_jacketSide_01_r, bone2Name=dyn_jacketSide_mstr_r, swing1LimitDegrees=35, swing2LimitDegrees=35, twistLimitDegrees=15}
 {bone1Name=dyn_jacketSide_02_r, bone2Name=dyn_jacketSide_01_r, swing1LimitDegrees=35, swing2LimitDegrees=35, twistLimitDegrees=15}
 {
  bone1Name=dyn_jacketSide_02_r
  bone2Name=dyn_jacketFront_r_02
  swing1LimitDegrees=45
  swing1Motion=Free
  swing2LimitDegrees=45
  swing2Motion=Free
  twistLimitDegrees=45
  twistMotion=Free
 }
 {
  bone1Name=dyn_jacketSide_02_l
  bone2Name=dyn_jacketFront_l_02
  swing1LimitDegrees=45
  swing1Motion=Free
  swing2LimitDegrees=45
  swing2Motion=Free
  twistLimitDegrees=45
  twistMotion=Free
 }
 {
  bone1Name=dyn_jacketBack_02_r
  bone2Name=dyn_jacketBack_02_l
  swing1LimitDegrees=45
  swing1Motion=Free
  swing2LimitDegrees=45
  swing2Motion=Free
  twistLimitDegrees=45
  twistMotion=Free
 }
 {bone1Name=dyn_jacketSide_03_l, bone2Name=dyn_jacketSide_02_l, swing1LimitDegrees=35, swing2LimitDegrees=35, twistLimitDegrees=15}
 {bone1Name=dyn_jacketSide_04_l, bone2Name=dyn_jacketSide_03_l, swing1LimitDegrees=35, swing2LimitDegrees=35, twistLimitDegrees=15}
 {bone1Name=dyn_jacketSide_03_r, bone2Name=dyn_jacketSide_02_r, swing1LimitDegrees=35, swing2LimitDegrees=35, twistLimitDegrees=15}
 {bone1Name=dyn_jacketSide_04_r, bone2Name=dyn_jacketSide_03_r, swing1LimitDegrees=35, swing2LimitDegrees=35, twistLimitDegrees=15}
 {
  bone1Name=dyn_jacketBack_03_r
  bone2Name=dyn_jacketSide_03_r
  swing1LimitDegrees=45
  swing1Motion=Free
  swing2LimitDegrees=45
  swing2Motion=Free
  twistLimitDegrees=45
  twistMotion=Free
 }
 {
  bone1Name=dyn_jacketBack_03_l
  bone2Name=dyn_jacketSide_03_l
  swing1LimitDegrees=45
  swing1Motion=Free
  swing2LimitDegrees=45
  swing2Motion=Free
  twistLimitDegrees=45
  twistMotion=Free
 }
]
