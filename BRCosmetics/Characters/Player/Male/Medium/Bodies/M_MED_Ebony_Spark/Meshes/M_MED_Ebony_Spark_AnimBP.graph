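# Cosmetic part AnimBP for the M_MED_Ebony_Spark body, animating jacket and straps over the pose copied from the parent mesh.
# Control Rigs, weighted constraints and movement/slide bone offsets feed RigidBody sims, tuned separately for skydiving.
# Straps pin to spine when physics is disabled; the floor collider drops while airborne.
# Events drive crouch, rifle and variant cape inputs, and hide a material element when an Ebony pickaxe is equipped outside skydiving.
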AnimBlueprint /BRCosmetics/Characters/Player/Male/Medium/Bodies/M_MED_Ebony_Spark/Meshes/M_MED_Ebony_Spark_AnimBP
target_skeleton=&/M_MED_Ebony_Spark_Skeleton
graphs [
 AnimGraph {nodes [
   "Transform (Modify) Bone - Bone: dyn_jacketBack_03_r"!(ComponentPose, bAlphaBoolEnabled:bool=True, Translation="0,0,0", Rotation="-25,-5,0", Scale="1,1,1") -> (Pose) {
    alphaInputType=Bool
    rotationMode=BMM_Additive
    rotationSpace=BCS_BoneSpace
    translationMode=BMM_Additive
    translationSpace=BCS_BoneSpace
    alphaBoolBlend {blendInTime=0.1, blendOption=Linear, blendOutTime=0.1}
    alphaScaleBiasClamp {bInterpResult=true}
    boneToModify {boneName=dyn_jacketBack_03_r}
    rotation {pitch=-25, yaw=-5}
   }
   "Transform (Modify) Bone - Bone: dyn_jacketSide_mstr_r"!(ComponentPose, bAlphaBoolEnabled:bool=True, Translation="0,0,0", Rotation="35,0,0", Scale="1,1,1") -> (Pose) {
    alphaInputType=Bool
    rotationMode=BMM_Additive
    rotationSpace=BCS_BoneSpace
    translationMode=BMM_Additive
    translationSpace=BCS_BoneSpace
    alphaBoolBlend {blendInTime=0.1, blendOption=Linear, blendOutTime=0.1}
    alphaScaleBiasClamp {bInterpResult=true}
    boneToModify {boneName=dyn_jacketSide_mstr_r}
    rotation {pitch=35}
   }
   "Transform (Modify) Bone - Bone: dyn_jacketBack_mstr_r"!(ComponentPose, bAlphaBoolEnabled:bool=True, Translation="0,0,0", Rotation="0,-10,0", Scale="1,1,1") -> (Pose) {
    alphaInputType=Bool
    rotationMode=BMM_Additive
    rotationSpace=BCS_BoneSpace
    translationMode=BMM_Additive
    translationSpace=BCS_BoneSpace
    alphaBoolBlend {blendInTime=0.1, blendOption=Linear, blendOutTime=0.1}
    alphaScaleBiasClamp {bInterpResult=true}
    boneToModify {boneName=dyn_jacketBack_mstr_r}
    rotation {yaw=-10}
   }
   n1 = Control Rig(Source=Copy Pose From Mesh() {bUseAttachedParent=true}) {
    alphaBoolBlend {blendOption=Linear}
    controlRigAssetReference {blueprintRigClass=/BRCosmetics/Characters/Player/Male/Medium/Bodies/M_MED_Ebony/Meshes/M_MED_Ebony_DeformRig_Master_ControlRig.M_MED_Ebony_DeformRig_Master_ControlRig_C}
   }
   n2 = "Constraint - dyn_jacketFront_mstr_l (PPPPPPPP)"!(
    ComponentPose=Local To Component(LocalPose=@n1)
    Alpha=1
    ConstraintWeights_0=0.4583
    ConstraintWeights_1:float=0.0827
    ConstraintWeights_2:float=0.1131
    ConstraintWeights_3:float=0.081
    ConstraintWeights_4:float=0.0258
    ConstraintWeights_5:float=0.0198
    ConstraintWeights_6:float=0.0222
    ConstraintWeights_7:float=0.1971
   ) -> (Pose) {
    alphaBoolBlend {blendOption=Linear}
    boneToModify {boneName=dyn_jacketFront_mstr_l}
    constraintSetup [
     {targetBone {boneName=pelvis}}
     {targetBone {boneName=spine_01}}
     {targetBone {boneName=spine_02}}
     {targetBone {boneName=spine_03}}
     {targetBone {boneName=thigh_twist_01_l}}
     {targetBone {boneName=deform_groin_l}}
     {targetBone {boneName=dyn_jacketFront_l_02}}
     {targetBone {boneName=dyn_jacketFront_l_01}}
    ]
    constraintWeights [1,1,1,1,1,1,1,1]
   }
   n3 = "Constraint - dyn_jacketFront_mstr_r (PPPPPPPP)"!(
    ComponentPose=@n2
    Alpha=1
    ConstraintWeights_0=0.62
    ConstraintWeights_1:float=0.0905
    ConstraintWeights_2:float=0.116
    ConstraintWeights_3:float=0.0753
    ConstraintWeights_4:float=0.0067
    ConstraintWeights_5:float=0.0383
    ConstraintWeights_6:float=0.0089
    ConstraintWeights_7:float=0.0443
   ) -> (Pose) {
    alphaBoolBlend {blendOption=Linear}
    boneToModify {boneName=dyn_jacketFront_mstr_r}
    constraintSetup [
     {targetBone {boneName=pelvis}}
     {targetBone {boneName=spine_01}}
     {targetBone {boneName=spine_02}}
     {targetBone {boneName=spine_03}}
     {targetBone {boneName=spine_04}}
     {targetBone {boneName=thigh_twist_01_r}}
     {targetBone {boneName=deform_glute_r}}
     {targetBone {boneName=deform_groin_r}}
    ]
    constraintWeights [1,1,1,1,1,1,1,1]
   }
   n4 = "Constraint - dyn_strap_fr_l_MSTR (PPPPPPPP)"!(
    ComponentPose=@n3
    Alpha=1
    ConstraintWeights_0=0.0407
    ConstraintWeights_1:float=0.1071
    ConstraintWeights_2:float=0.2445
    ConstraintWeights_3:float=0.2214
    ConstraintWeights_4:float=0.1618
    ConstraintWeights_5:float=0.0115
    ConstraintWeights_6:float=0.0142
    ConstraintWeights_7:float=0.1987
   ) -> (Pose) {
    alphaBoolBlend {blendOption=Linear}
    boneToModify {boneName=dyn_strap_fr_l_MSTR}
    constraintSetup [
     {targetBone {boneName=spine_02}}
     {targetBone {boneName=spine_03}}
     {targetBone {boneName=spine_04}}
     {targetBone {boneName=spine_05}}
     {targetBone {boneName=clavicle_l}}
     {targetBone {boneName=upperarm_twist_01_l}}
     {targetBone {boneName=deform_upperarm_fr_l}}
     {targetBone {boneName=deform_pec_l}}
    ]
    constraintWeights [1,1,1,1,1,1,1,1]
   }
   n5 = "Constraint - dyn_strap_fr_r_MSTR (PPPPPPPP)"!(
    ComponentPose=@n4
    Alpha=1
    ConstraintWeights_0=0.0407
    ConstraintWeights_1:float=0.1071
    ConstraintWeights_2:float=0.2445
    ConstraintWeights_3:float=0.2214
    ConstraintWeights_4:float=0.1618
    ConstraintWeights_5:float=0.0115
    ConstraintWeights_6:float=0.0142
    ConstraintWeights_7:float=0.1987
   ) -> (Pose) {
    alphaBoolBlend {blendOption=Linear}
    boneToModify {boneName=dyn_strap_fr_r_MSTR}
    constraintSetup [
     {targetBone {boneName=spine_02}}
     {targetBone {boneName=spine_03}}
     {targetBone {boneName=spine_04}}
     {targetBone {boneName=spine_05}}
     {targetBone {boneName=clavicle_r}}
     {targetBone {boneName=upperarm_twist_01_r}}
     {targetBone {boneName=deform_upperarm_fr_r}}
     {targetBone {boneName=deform_pec_r}}
    ]
    constraintWeights [1,1,1,1,1,1,1,1]
   }
   n6 = Control Rig!(
    Source=Component To Local(ComponentPose=@n5)
    Alpha
    crouching:double=Get crouch() -> (crouch:double)
    cape:double=Get Cape() -> (Cape:double)
    rban disabled:double=Get rban disabled() -> (rban disabled:double)
   ) -> (Pose) {
    alphaBoolBlend {blendOption=Linear}
    controlRigAssetReference {blueprintRigClass=&/M_MED_Ebony_Spark_CtrlRig.M_MED_Ebony_Spark_CtrlRig_C}
   }
   n7 = Get bIsMoving2D()
   n8 = "Transform (Modify) Bone - Bone: dyn_jacketBack_mstr_l"!(ComponentPose=Local To Component(LocalPose=@n6), bAlphaBoolEnabled:bool=@n7, Translation="0,0,1", Rotation="15,0,0", Scale="1,1,1") -> (Pose) {
    alphaInputType=Bool
    rotationMode=BMM_Additive
    rotationSpace=BCS_BoneSpace
    translationMode=BMM_Additive
    translationSpace=BCS_BoneSpace
    alphaBoolBlend {blendInTime=0.1, blendOption=Linear, blendOutTime=0.1}
    alphaScaleBiasClamp {bInterpResult=true}
    boneToModify {boneName=dyn_jacketBack_mstr_l}
    rotation {pitch=15}
    translation {z=1}
   }
   n9 = "Transform (Modify) Bone - Bone: dyn_jacketBack_mstr_r"(ComponentPose=@n8, bAlphaBoolEnabled=@n7, Translation="0,0,1", Rotation="15,0,0", Scale="1,1,1") {
    alphaInputType=Bool
    rotationMode=BMM_Additive
    rotationSpace=BCS_BoneSpace
    translationMode=BMM_Additive
    translationSpace=BCS_BoneSpace
    alphaBoolBlend {blendInTime=0.1, blendOption=Linear, blendOutTime=0.1}
    alphaScaleBiasClamp {bInterpResult=true}
    boneToModify {boneName=dyn_jacketBack_mstr_r}
    rotation {pitch=15}
    translation {z=1}
   }
   n10 = "Transform (Modify) Bone - Bone: dyn_jacketSide_mstr_r"(ComponentPose=@n9, bAlphaBoolEnabled=@n7, Translation="0,0,0", Rotation="10,0,0", Scale="1,1,1") {
    alphaInputType=Bool
    rotationMode=BMM_Additive
    rotationSpace=BCS_BoneSpace
    translationMode=BMM_Additive
    translationSpace=BCS_BoneSpace
    alphaBoolBlend {blendInTime=0.1, blendOption=Linear, blendOutTime=0.1}
    alphaScaleBiasClamp {bInterpResult=true}
    boneToModify {boneName=dyn_jacketSide_mstr_r}
    rotation {pitch=10}
   }
   n11 = "Transform (Modify) Bone - Bone: dyn_jacketSide_mstr_l"!(ComponentPose=@n10, bAlphaBoolEnabled:bool=@n7, Translation="0,0,0", Rotation="10,0,0", Scale="1,1,1") -> (Pose) {
    alphaInputType=Bool
    rotationMode=BMM_Additive
    rotationSpace=BCS_BoneSpace
    translationMode=BMM_Additive
    translationSpace=BCS_BoneSpace
    alphaBoolBlend {blendInTime=0.1, blendOption=Linear, blendOutTime=0.1}
    alphaScaleBiasClamp {bInterpResult=true}
    boneToModify {boneName=dyn_jacketSide_mstr_l}
    rotation {pitch=10}
   }
   n12 = Get bIsSliding()
   n13 = "Transform (Modify) Bone - Bone: dyn_jacketBack_mstr_l"(ComponentPose=@n11, bAlphaBoolEnabled=@n12, Translation="0,0,3", Rotation="55,0,0", Scale="1,1,1") {
    alphaInputType=Bool
    rotationMode=BMM_Additive
    rotationSpace=BCS_BoneSpace
    translationMode=BMM_Additive
    translationSpace=BCS_BoneSpace
    alphaBoolBlend {blendInTime=0.1, blendOption=Linear, blendOutTime=0.1}
    alphaScaleBiasClamp {bInterpResult=true}
    boneToModify {boneName=dyn_jacketBack_mstr_l}
    rotation {pitch=55}
    translation {z=3}
   }
   n14 = "Transform (Modify) Bone - Bone: dyn_jacketBack_mstr_r"(ComponentPose=@n13, bAlphaBoolEnabled=@n12, Translation="0,0,5", Rotation="55,0,0", Scale="1,1,1") {
    alphaInputType=Bool
    rotationMode=BMM_Additive
    rotationSpace=BCS_BoneSpace
    translationMode=BMM_Additive
    translationSpace=BCS_BoneSpace
    alphaBoolBlend {blendInTime=0.1, blendOption=Linear, blendOutTime=0.1}
    alphaScaleBiasClamp {bInterpResult=true}
    boneToModify {boneName=dyn_jacketBack_mstr_r}
    rotation {pitch=55}
    translation {z=5}
   }
   n15 = "Transform (Modify) Bone - Bone: dyn_jacketSide_mstr_r"(ComponentPose=@n14, bAlphaBoolEnabled=@n12, Translation="0,0,0", Rotation="35,0,0", Scale="1,1,1") {
    alphaInputType=Bool
    rotationMode=BMM_Additive
    rotationSpace=BCS_BoneSpace
    translationMode=BMM_Additive
    translationSpace=BCS_BoneSpace
    alphaBoolBlend {blendInTime=0.1, blendOption=Linear, blendOutTime=0.1}
    alphaScaleBiasClamp {bInterpResult=true}
    boneToModify {boneName=dyn_jacketSide_mstr_r}
    rotation {pitch=35}
   }
   n16 = "Transform (Modify) Bone - Bone: dyn_jacketSide_mstr_l"(ComponentPose=@n15, bAlphaBoolEnabled=@n12, Translation="0,0,0", Rotation="35,0,0", Scale="1,1,1") {
    alphaInputType=Bool
    rotationMode=BMM_Additive
    rotationSpace=BCS_BoneSpace
    translationMode=BMM_Additive
    translationSpace=BCS_BoneSpace
    alphaBoolBlend {blendInTime=0.1, blendOption=Linear, blendOutTime=0.1}
    alphaScaleBiasClamp {bInterpResult=true}
    boneToModify {boneName=dyn_jacketSide_mstr_l}
    rotation {pitch=35}
   }
   n17 = NOT Boolean(A=Get bRigidBodyEnabled())
   n18 = "Constraint - dyn_strap_fr_l_02 (P)"!(ComponentPose=@n16, bAlphaBoolEnabled:bool=@n17, ConstraintWeights_0=1) -> (Pose) {
    alphaInputType=Bool
    alphaBoolBlend {blendInTime=0.1, blendOption=Linear, blendOutTime=0.1}
    boneToModify {boneName=dyn_strap_fr_l_02}
    constraintSetup [{targetBone {boneName=spine_04}}]
    constraintWeights [1]
   }
   n19 = "Constraint - dyn_strap_fr_l_03 (P)"!(ComponentPose=@n18, bAlphaBoolEnabled:bool=@n17, ConstraintWeights_0=1) -> (Pose) {
    alphaInputType=Bool
    alphaBoolBlend {blendInTime=0.1, blendOption=Linear, blendOutTime=0.1}
    boneToModify {boneName=dyn_strap_fr_l_03}
    constraintSetup [{targetBone {boneName=spine_03}}]
    constraintWeights [1]
   }
   n20 = "Constraint - dyn_strap_fr_l_04 (P)"!(ComponentPose=@n19, bAlphaBoolEnabled:bool=@n17, ConstraintWeights_0=1) -> (Pose) {
    alphaInputType=Bool
    alphaBoolBlend {blendInTime=0.1, blendOption=Linear, blendOutTime=0.1}
    boneToModify {boneName=dyn_strap_fr_l_04}
    constraintSetup [{targetBone {boneName=spine_02}}]
    constraintWeights [1]
   }
   n21 = "Constraint - dyn_strap_fr_r_02 (P)"!(ComponentPose=@n20, bAlphaBoolEnabled:bool=@n17, ConstraintWeights_0=1) -> (Pose) {
    alphaInputType=Bool
    alphaBoolBlend {blendInTime=0.1, blendOption=Linear, blendOutTime=0.1}
    boneToModify {boneName=dyn_strap_fr_r_02}
    constraintSetup [{targetBone {boneName=spine_04}}]
    constraintWeights [1]
   }
   n22 = "Constraint - dyn_strap_fr_r_03 (P)"!(ComponentPose=@n21, bAlphaBoolEnabled:bool=@n17, ConstraintWeights_0=1) -> (Pose) {
    alphaInputType=Bool
    alphaBoolBlend {blendInTime=0.1, blendOption=Linear, blendOutTime=0.1}
    boneToModify {boneName=dyn_strap_fr_r_03}
    constraintSetup [{targetBone {boneName=spine_03}}]
    constraintWeights [1]
   }
   n23 = "Constraint - dyn_strap_fr_r_04 (P)"!(ComponentPose=@n22, bAlphaBoolEnabled:bool=@n17, ConstraintWeights_0=1) -> (Pose) {
    alphaInputType=Bool
    alphaBoolBlend {blendInTime=0.1, blendOption=Linear, blendOutTime=0.1}
    boneToModify {boneName=dyn_strap_fr_r_04}
    constraintSetup [{targetBone {boneName=spine_02}}]
    constraintWeights [1]
   }
   n24 = "Transform (Modify) Bone - Bone: Floor_COL"(ComponentPose=@n23, Alpha=Get Floor COL() -> (Floor COL:double), Translation="0,0,-100", Rotation="0,0,0", Scale="1,1,1") {
    translationMode=BMM_Additive
    translationSpace=BCS_BoneSpace
    alphaBoolBlend {blendOption=Linear}
    alphaScaleBiasClamp {bInterpResult=true}
    boneToModify {boneName=Floor_COL}
    translation {z=-100}
   }
   n25 = Get bIsSkyDiving()
   n26 = RigidBody!(ComponentPose=@n24, bAlphaBoolEnabled:bool=@n25, OverrideWorldGravity:Vector=Get GravityOverride(), ExternalForce) -> (Pose) {
    alphaInputType=Bool
    bClampLinearTranslationLimitToRefPose=true
    bOverrideWorldGravity=true
    overridePhysicsAsset=&/M_MED_Ebony_Spark_Physics
    simulationSpace=BaseBoneSpace
    alphaBoolBlend {blendInTime=0.1, blendOption=Linear, blendOutTime=0.1}
    baseBoneRef {boneName=pelvis}
    componentAppliedLinearAccClamp=<3000 3000 3000>
    componentLinearAccScale=<0.25 0.25 0.4>
    componentLinearVelScale=<0.2 0.2 0.2>
   }
   n27 = Get GravityOverride()
   n28 = RigidBody(ComponentPose=@n26, bAlphaBoolEnabled=NOT Boolean(A=@n25), OverrideWorldGravity=@n27) {
    alphaInputType=Bool
    bClampLinearTranslationLimitToRefPose=true
    bOverrideWorldGravity=true
    overridePhysicsAsset=&M_MED_Ebony_Spark_Physics
    simulationSpace=BaseBoneSpace
    alphaBoolBlend {blendInTime=0.1, blendOption=Linear, blendOutTime=0.1}
    baseBoneRef {boneName=pelvis}
    componentAppliedLinearAccClamp=<3000 3000 3000>
    componentLinearAccScale=<0.5 0.3 0.3>
    componentLinearVelScale=<0.5 0.3 0.3>
    overrideWorldGravity {z=-980}
    simSpaceSettings {
     maxAngularAcceleration=1000
     maxAngularVelocity=3
     maxLinearAcceleration=700
     maxLinearVelocity=500
     velocityScaleZ=0
     worldAlpha=1
     externalLinearDragV=<0.2 0.2>
    }
   }
   n29 = RigidBody!(ComponentPose=@n28, Alpha:float, OverrideWorldGravity=@n27, ExternalForce) -> (Pose) {
    bClampLinearTranslationLimitToRefPose=true
    bOverrideWorldGravity=true
    overridePhysicsAsset=&/M_MED_Ebony_Spark_Front_Physics
    simulationSpace=BaseBoneSpace
    alphaBoolBlend {blendOption=Linear}
    baseBoneRef {boneName=pelvis}
    componentAppliedLinearAccClamp=<3000 3000 3000>
    componentLinearAccScale=<0.3 0.3 0.3>
    componentLinearVelScale=<0.3 0.3 0.3>
   }
   n30 = "Transform (Modify) Bone - Bone: dyn_jacketSide_03_r"(ComponentPose=@n29, Alpha=Get jacket() -> (jacket:double), Translation="0,0,0", Rotation="20,-30,-35", Scale="1,1,1") {
    rotationMode=BMM_Additive
    rotationSpace=BCS_BoneSpace
    translationMode=BMM_Additive
    translationSpace=BCS_BoneSpace
    alphaBoolBlend {blendInTime=0.1, blendOption=Linear, blendOutTime=0.1}
    alphaScaleBiasClamp {bInterpResult=true}
    boneToModify {boneName=dyn_jacketSide_03_r}
    rotation=<rot 20 -30 -35>
   }
   n31 = "Transform (Modify) Bone - Bone: dyn_underarm_fix"!(ComponentPose=@n30, bAlphaBoolEnabled:bool=Get underarm() -> (underarm:bool), Translation="0,0,-4", Rotation="0,0,0", Scale="1,1,1") -> (Pose) {
    alphaInputType=Bool
    rotationMode=BMM_Additive
    rotationSpace=BCS_BoneSpace
    translationMode=BMM_Additive
    translationSpace=BCS_BoneSpace
    alphaBoolBlend {blendInTime=0.1, blendOption=Linear, blendOutTime=0.1}
    alphaScaleBiasClamp {bInterpResult=true}
    boneToModify {boneName=dyn_underarm_fix}
    translation {z=-4}
   }
   n32 = "Transform (Modify) Bone - Bone: dyn_strap_fr_r_MSTR"(ComponentPose=@n31, Alpha=1, Translation="0,0,0", Rotation=Get back_rotate() -> (back_rotate:Rotator), Scale="1,1,1") {
    rotationMode=BMM_Additive
    rotationSpace=BCS_BoneSpace
    translationMode=BMM_Additive
    translationSpace=BCS_BoneSpace
    alphaBoolBlend {blendInTime=0.1, blendOption=Linear, blendOutTime=0.1}
    alphaScaleBiasClamp {bInterpResult=true}
    boneToModify {boneName=dyn_strap_fr_r_MSTR}
    rotation {yaw=3}
   }
   n33 = Control Rig!(Source=Component To Local(ComponentPose=@n32), Alpha, rifle:double=Get Rifle_alpha() -> (Rifle_alpha:double), Cape:double=Get Cape()) -> (Pose) {
    alphaBoolBlend {blendOption=Linear}
    controlRigAssetReference {blueprintRigClass=/BRCosmetics/Characters/Player/Male/Medium/Bodies/M_MED_Ebony/Meshes/M_MED_Ebony_post_CtrlRig.M_MED_Ebony_post_CtrlRig_C}
   }
   Output Pose(Result=@n33) {layerGroup=None}
  ]}
 EventGraph {nodes [
   Event BlueprintUpdateAnimation()
   > n1 = Sequence!(execute) -> (then_0, then_1, then_2:Exec, then_3:Exec)
   > n2 = Gravity Override State Function DEPRECATED(CharacterPartAnimInstance=Self-Reference(), GameTimeInSeconds=GetGameTimeInSeconds(), GravityOverrideParams=Get Gravity Override Params())
   > Set GravityOverride(GravityOverride=@n2)
   > Set underarm(underarm:bool=true) -> (Output_Get:bool)
   n3 = GetOwnerSkeletalMeshComponent()
   n4 = BreakRotator(InRot=BreakTransform(InTransform=GetDeltaTransformFromRefPose(self=@n3, BoneName=spine_05, BaseName=spine_01)).Rotation)
   > Set back_rotate(back_rotate:Rotator=MakeRotator(Yaw=MapRangeClamped(Value=@n4.Yaw, InRangeA=5, InRangeB=30, OutRangeB=3))) -> (Output_Get:Rotator)
   > n5 = Branch(Condition=Get bRigidBodyEnabled())
   > Set rban disabled(rban disabled:double=0) -> (Output_Get:double)
   Set rban disabled(execute=@n5.else, rban disabled=1)
   n6 = Get bIsMoving2D()
   n7 = Branch(execute=@n1.then_1, Condition=AND Boolean!(A=Get bIsCrouching(), B=NOT Boolean(A=@n6), C:bool=Get bRigidBodyEnabled()) -> (ReturnValue))
   > Set jacket(jacket:double=1) -> (Output_Get:double)
   Set jacket(execute=@n7.else, jacket=0)
   n8 = Cast To FortPlayerAnimInstance(execute=@n1.then_2, Object=GetAnimInstance(self=GetOwnerSkeletalMeshComponent()))
   > Set Rifle(Rifle:bool=false) -> (Output_Get:bool)
   > n9 = Switch on EFortWeaponCoreAnimation!(execute, Selection:EFortWeaponCoreAnimation Enum=Get WeaponCoreAnim(self=@n8)) -> (
    Melee:Exec
    Pistol:Exec
    Shotgun:Exec
    PaperBlueprint:Exec
    Rifle:Exec
    MeleeOneHand:Exec
    MachinePistol:Exec
    RocketLauncher:Exec
    GrenadeLauncher:Exec
    GoingCommando:Exec
    AssaultRifle:Exec
    TacticalShotgun:Exec
    SniperRifle:Exec
    TrapPlacement:Exec
    ShoulderLauncher:Exec
    AbilityDecoTool:Exec
    Crossbow:Exec
    C4:Exec
    RemoteControl:Exec
    DualWield:Exec
    AR_BullPup:Exec
    AR_ForwardGrip:Exec
    MedPackPaddles:Exec
    SMG_P90:Exec
    AR_DrumGun:Exec
    Consumable_Small:Exec
    Consumable_Large:Exec
    Balloon:Exec
    MountedTurret:Exec
    CreativeTool:Exec
    ExplosiveBow:Exec
    AshtonIndigo:Exec
    AshtonChicago:Exec
    MeleeDualWield:Exec
    Unarmed:Exec
   )
   Branch(execute=[
     @n9.AR_BullPup
     @n9.AR_ForwardGrip
     @n9.AssaultRifle
     @n9.Crossbow
     @n9.GrenadeLauncher
     @n9.Rifle
     @n9.RocketLauncher
     @n9.Shotgun
     @n9.SniperRifle
     @n9.TacticalShotgun
    ])
   > Set Rifle(Rifle=true)
   n10 = Branch(execute=@n1.then_3, Condition=Get Rifle() -> (Rifle:bool))
   > Set Rifle_alpha(Rifle_alpha:double=1) -> (Output_Get:double)
   Set Rifle_alpha(execute=@n10.else, Rifle_alpha=0)
   n11 = Event OnBeginJumping()
   Set Floor COL(execute=[@n11.then, @n19.then], Floor COL:double=1) -> (Output_Get:double)
   Event OnEndCrouching()
   > Set crouch(crouch:double=0) -> (Output_Get:double)
   Event OnCharacterCustomizationLoaded()
   > n12 = Sequence!(execute) -> (then_0, then_1)
   n13 = Get Dynamic Material Instance(execute=[@n23.then, @n12.then_0], SkeletalMeshComponent=GetPartSkeletalMeshComponent())
   n14 = PickaxeItemDefinitionHasAnyMetaTag(Tags=Make Array("[0]"=(TagName="Cosmetics.MetaTags.Ebony")), bRequiresEquipped=true)
   > n15 = Branch(Condition=@n14)
   > n16 = Branch(Condition=Get bIsSkyDiving())
   SetScalarParameterValue(execute=[@n20.then, @n21.else, @n15.else, @n16.then], self=@n13, ParameterName=Hide Element 02, Value=0)
   SetScalarParameterValue(execute=[@n21.then, @n16.else], self=@n13, ParameterName=Hide Element 02, Value=1)
   n17 = Branch(execute=@n12.then_1, Condition=CharacterItemDefinitionHasAnyMetaTag(Tags=Make Array("[0]"=(TagName="Cosmetics.Variant.Property.Stage3"))))
   > Set Cape(Cape:double=1) -> (Output_Get:double)
   Set Cape(execute=@n17.else, Cape=0)
   n18 = Event OnEndJumping()
   Set Floor COL(execute=[@n18.then, @n22.then], Floor COL=0)
   n19 = Event OnBeginFalling()
   n20 = Event OnBeginSkydiving()
   Event OnEndSkydiving()
   > n21 = Branch(Condition=@n14)
   n22 = Event OnEndFalling()
   Event OnBeginCrouching()
   > Set crouch(crouch=1)
   n23 = Event OnWeaponChanged()
   Event OnBeginKneeSliding() {is=title K2Node_Event}
   Event OnBeginBaseballSliding() {is=title K2Node_Event}
   Event OnEndBaseballSliding() {is=title K2Node_Event}
   Event OnEndKneeSliding() {is=title K2Node_Event}
  ]}
]
variables [
 back_rotate {type=Rotator}
 Cape {type=number}
 crouch {type=number}
 Floor COL {type=number}
 Gravity Override Params {type=GravityOverrideParamsStruct}
 GravityOverride {type=Vector}
 jacket {type=number}
 rban disabled {type=number}
 Rifle {type=boolean}
 Rifle_alpha {type=number}
 underarm {type=boolean}
]
